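# Conversation database for the TNTina NPC, defining what she says when a player talks to her and which shared dialogue branches follow.
# Its entry point leads to one speech node with a default line and contextual lines gated by first-ever conversation, player outfit or match conversation count.
# Curve rows set priority for first-meeting and outfit lines; count-gated lines name none.
# The speech node then links to the shared NPC Quests and BasicServices entries.
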ConversationDatabase /NPCLibrary/NPCs/TNTina/NPCLibrary_Conversation_TNTina
entry_points [{tag=Conversation.Entry.TNTina, destinations [#1]}]
nodes [
 ConversationEntryPointNode_7 {
  id=#1
  class=ConversationEntryPointNode
  outputs [#2]
  properties {entryTag {tagName=Conversation.Entry.TNTina}}
 }
 Speech_C_4 {
  id=#2
  class=Speech_C
  outputs [#3, #4]
  properties {
   nodeName=""
   generalConfig {
    defaultMessage="I just wanna make things go BOOM!"
    contextualMessages [
     {
      message="Heya, I'm TNTina. The original short fuse."
      contextRequirements [{requirement {class=FortControllerRequirement_IsFirstEverConversationWithNPC}}]
      priority {curve {rowName=Tandem.Speech.Priority.FirstEverInteraction}}
     }
     {
      message="Getting the gang back together's really lit my fuse. "
      contextRequirements [{requirement {
         class=FortControllerRequirement_HasCID
         properties {allowedCIds [
           &CID_693_Athena_Commando_M_BuffCat
           /BRCosmetics/Athena/Items/Cosmetics/Characters/CID_770_Athena_Commando_F_MechanicalEngineer
           &CID_692_Athena_Commando_M_HenchmanTough
           /BRCosmetics/Athena/Items/Cosmetics/Characters/CID_690_Athena_Commando_F_Photographer
           /BRCosmetics/Athena/Items/Cosmetics/Characters/CID_A_034_Athena_Commando_F_CatBurglar
          ]}
        }}]
      priority {curve {rowName=Tandem.Speech.Priority.OutfitReaction}}
     }
     {
      message=Thanks for sending up a flare.
      contextRequirements [{requirement {
         class=FortControllerRequirement_HasCID
         properties {allowedCIds [&CID_A_034_Athena_Commando_F_CatBurglar]}
        }}]
      priority {curve {rowName=Tandem.Speech.Priority.OutfitReaction}}
     }
     {
      message="You were always more gunpowder than fuse, boss."
      contextRequirements [{requirement {class=FortControllerRequirement_HasCID, properties {allowedCIds [/BRCosmetics/Athena/Items/Cosmetics/Characters/Character_PiteousKicks]}}}]
      priority {curve {rowName=Tandem.Speech.Priority.OutfitReaction}}
     }
     {
      message="I just wanna make things go BOOM!"
      contextRequirements [{requirement {class=FortControllerRequirement_CheckNPCMatchConversationCount, properties {nPCUniqueIdTag {tagName=AISpawnerData.Type.Tandem.TNTina}}}}]
      priority {curve {rowName=None}}
     }
     {
      message=It's nice to be back.
      contextRequirements [{requirement {class=FortControllerRequirement_CheckNPCMatchConversationCount, properties {nPCUniqueIdTag {tagName=AISpawnerData.Type.Tandem.TNTina}}}}]
      priority {curve {rowName=None}}
     }
     {
      message="Looking forward to seeing the team back together. "
      contextRequirements [{requirement {class=FortControllerRequirement_CheckNPCMatchConversationCount, properties {nPCUniqueIdTag {tagName=AISpawnerData.Type.Tandem.TNTina}}}}]
      priority {curve {rowName=None}}
     }
     {
      message="Stick with me, I bet we could really blow up."
      contextRequirements [{requirement {class=FortControllerRequirement_CheckNPCMatchConversationCount, properties {nPCUniqueIdTag {tagName=AISpawnerData.Type.Tandem.TNTina}}}}]
      priority {curve {rowName=None}}
     }
     {
      message="GRENADE!! Ha, just kidding."
      contextRequirements [{requirement {class=FortControllerRequirement_CheckNPCMatchConversationCount, properties {nPCUniqueIdTag {tagName=AISpawnerData.Type.Tandem.TNTina}}}}]
      priority {curve {rowName=None}}
     }
    ]
   }
  }
 }
 ConversationLinkNode_0 {id=#4, class=ConversationLinkNode, properties {remoteEntryTag {tagName=Conversation.Entry.BasicServices}}}
 ConversationLinkNode_1 {id=#3, class=ConversationLinkNode, properties {remoteEntryTag {tagName=Conversation.Entry.NPCLibrary.Quests}}}
]
